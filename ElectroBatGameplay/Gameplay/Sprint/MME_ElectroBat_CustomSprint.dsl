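; Movement mode extension for the ElectroBat's custom sprint, which records the weapon that activated it.
; On activation, on a dedicated server only, it fetches the item-activated runtime data from the runtime context.
; If the owner pawn has the ElectroBat equipped tag, it sets the pawn's current weapon as the activating item.
; Otherwise it logs that the tag is missing and the item cannot be set reliably.
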(event EventOnExtensionActivate
  (params (RuntimeContext :type MMERuntime%20Context%20Structure%20%28by%20ref%29))
  (bind _ownerpawn (Variables|FortMovementMode_BaseExtLogic|GetOwnerPawn))
  ;[0.7899 0.1002 0.2833] Apply a tag to the player when they tac sprint jump
  (if (CallFunction|IsServer?)
    (bind _returnvalue (Class|FortMovementModeExtBlueprintLibrary|GetRuntimeDatafromContext RuntimeContext "/Script/FortniteGame.FortMovementMode_ItemActivatedRuntimeData"))
    (if (GameplayTags|HasMatchingGameplayTag _ownerpawn (Variables|Default|GetElectroBatEquippedTag))
      (Class|FortMovementModeItemActivatedRuntimeData|SetActivatingItemfromWeapon _returnvalue (Variables|/Script/FortniteGame.FortPawn|GetCurrentWeapon _ownerpawn))
      (else
        (Fortnite|FortShippingLog _ownerpawn "MME_ElectroBat_CustomSprint activated but the ElectroBat Equipped tag is missing from the OwnerPawn! Cannot set activating item from weapon reliably." true)))))

(fn IsServer?
  (return (Networking|IsDedicatedServer (Variables|FortMovementMode_BaseExtLogic|GetOwnerPawn))))
